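# Sound cue for the exterior active ambient bed loop in the Mother event's warm-up phase, under Athena.
# A single looping wave player plays the Mother active ambient bed sound wave with no other processing nodes.
# It routes through the Athena outdoor ambient sound class at reduced volume, with a very large maximum audible distance.
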SoundCue /Game/Athena/Sounds/Events/Mother/Warm_Up/Mother_Ext_Warm_Up_Active_Ambient_Bed_Loop
root=WavePlayer_0
nodes [{id=WavePlayer_0, properties {bLooping=true, soundWaveAssetPtr=&/Mother_Active_Ambient_Bed_Loop}}]
properties {maxDistance=2097152, soundClassObject=&Fort_Ambient_Outdoors_Athena, volumeMultiplier=0.376}
