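# Sound cue for the exterior "Warm Up Open" sound of the Mother event in Athena, routed to the in-game SFX class with boosted volume.
# A Mixer layers a 3D wave through its own attenuation node, at slightly reduced level, with a 2D wave that bypasses that node.
# A concurrency override allows only one instance and blocks new plays while it is active.
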SoundCue /Game/Athena/Sounds/Events/Mother/Warm_Up/Mother_Ext_Warm_Up_Open
root=Mixer_0
nodes [
 {
  id=Attenuation_0
  children [WavePlayer_0]
  properties {attenuationSettings=&/Mother_Ext_Warm_Up_Open_Att}
 }
 {id=Mixer_0, children [Attenuation_0, WavePlayer_1], properties {inputVolume [0.75,1]}}
 {id=WavePlayer_0, properties {soundWaveAssetPtr=&/Mother_Warm_Up_Open_3D_01}}
 {id=WavePlayer_1, properties {soundWaveAssetPtr=&/Mother_Warm_Up_Open_2D_01}}
]
properties {
 bOverrideConcurrency=true
 maxDistance=2097152
 soundClassObject=&Fort_SFX_InGame
 volumeMultiplier=4
 concurrencyOverrides {maxCount=1, resolutionRule=PreventNew}
}
